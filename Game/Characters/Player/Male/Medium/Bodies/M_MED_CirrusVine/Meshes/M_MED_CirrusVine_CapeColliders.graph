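# Physics asset of cape colliders for the CirrusVine medium male player body, presumably used for cape simulation.
# Capsules on pelvis, spine, clavicles and neck plus a head sphere cover the torso; arm and leg bodies have no shapes.
# Four drape chains (left and right, back and side) hang from the pelvis, each with four capsules offset from their bones.
# Constraints follow the bone chains with swing and twist left free.
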PhysicsAsset /Game/Characters/Player/Male/Medium/Bodies/M_MED_CirrusVine/Meshes/M_MED_CirrusVine_CapeColliders
bodies {
 calf_l []
 calf_r []
 clavicle_l [{shapeType=Capsule, length=7.7636, radius=11.1427, center=<9.7517 2.049 -1.5391>, rotation=<rot 75.9352 -0.0001 175.5622>}]
 clavicle_r [{shapeType=Capsule, length=7.7636, radius=11.1427, center=<-9.7517 -2.049 1.5391>, rotation=<rot 75.9351 -0.0001 -4.4378>}]
 dyn_C_drape_l_back_1 [{shapeType=Capsule, length=15, radius=10, shapeName=dyn_C_drape_l_back_1_capsule, center=<-6 -6 0>}]
 dyn_C_drape_l_back_2 [{shapeType=Capsule, length=15, radius=10, shapeName=dyn_C_drape_l_back_2_capsule, center=<-6 -6 0>}]
 dyn_C_drape_l_back_3 [{shapeType=Capsule, length=15, radius=10, shapeName=dyn_C_drape_l_back_3_capsule, center=<-6 -6 0>}]
 dyn_C_drape_l_back_4 [{shapeType=Capsule, length=15, radius=10, shapeName=dyn_C_drape_l_back_4_capsule, center=<-6 -6 0>}]
 dyn_C_drape_r_back_1 [{shapeType=Capsule, length=15, radius=10, shapeName=dyn_C_drape_r_back_1_capsule, center=<6 6 0>}]
 dyn_C_drape_r_back_2 [{shapeType=Capsule, length=15, radius=10, shapeName=dyn_C_drape_r_back_2_capsule, center=<6 6 0>}]
 dyn_C_drape_r_back_3 [{shapeType=Capsule, length=15, radius=10, shapeName=dyn_C_drape_r_back_3_capsule, center=<6 6 0>}]
 dyn_C_drape_r_back_4 [{shapeType=Capsule, length=15, radius=10, shapeName=dyn_C_drape_r_back_4_capsule, center=<6 6 0>}]
 dyn_L_drape_side_1 [{shapeType=Capsule, length=15, radius=10, shapeName=dyn_L_drape_side_1_capsule, center=<-6 6 0>}]
 dyn_L_drape_side_2 [{shapeType=Capsule, length=15, radius=10, shapeName=dyn_L_drape_side_2_capsule, center=<-6 6 0>}]
 dyn_L_drape_side_3 [{shapeType=Capsule, length=15, radius=10, shapeName=dyn_L_drape_side_3_capsule, center=<-6 6 0>}]
 dyn_L_drape_side_4 [{shapeType=Capsule, length=15, radius=10, shapeName=dyn_L_drape_side_4_capsule, center=<-6 6 0>}]
 dyn_R_drape_side_1 [{shapeType=Capsule, length=15, radius=10, shapeName=dyn_R_drape_side_1_capsule, center=<6 -6 0>}]
 dyn_R_drape_side_2 [{shapeType=Capsule, length=15, radius=10, shapeName=dyn_R_drape_side_2_capsule, center=<6 -6 0>}]
 dyn_R_drape_side_3 [{shapeType=Capsule, length=15, radius=10, shapeName=dyn_R_drape_side_3_capsule, center=<6 -6 0>}]
 dyn_R_drape_side_4 [{shapeType=Capsule, length=15, radius=10, shapeName=dyn_R_drape_side_4_capsule, center=<6 -6 0>}]
 foot_l []
 foot_r []
 head [{shapeType=Sphere, radius=17.3125, center=<11.028 -0.174 0>}]
 lowerarm_l []
 lowerarm_r []
 neck_02 [{shapeType=Capsule, length=13.4375, radius=7.3175, center=<-0.2642 -0.3408 0>, rotation=<rot 90 -171.1193 -175.3978>}]
 pelvis [{shapeType=Capsule, length=8.6562, radius=13.5388, center=<-2.6215 -1.3695 0>}]
 spine_01 [{shapeType=Capsule, length=10, radius=12.4138, center=<-2.6215 -0.7156 0>}]
 spine_02 [{shapeType=Capsule, length=8, radius=13.195, center=<-2.6215 0.4993 0.0001>}]
 spine_03 [{shapeType=Capsule, length=10, radius=11.7575, center=<-1.953 -0.9223 0.0001>}]
 spine_04 [{shapeType=Capsule, length=11, radius=14.1325, center=<2.1467 1.5797 0.0002>}]
 spine_05 [{shapeType=Capsule, length=11, radius=15, center=<-1.8422 1 0>}]
 thigh_l []
 thigh_r []
 upperarm_l []
 upperarm_r []
}
constraints [
 {
  bone1Name=spine_01
  bone2Name=pelvis
  swing1LimitDegrees=45
  swing1Motion=Free
  swing2LimitDegrees=45
  swing2Motion=Free
  twistLimitDegrees=45
  twistMotion=Free
 }
 {
  bone1Name=spine_02
  bone2Name=spine_01
  swing1LimitDegrees=45
  swing1Motion=Free
  swing2LimitDegrees=45
  swing2Motion=Free
  twistLimitDegrees=45
  twistMotion=Free
 }
 {
  bone1Name=spine_03
  bone2Name=spine_02
  swing1LimitDegrees=45
  swing1Motion=Free
  swing2LimitDegrees=45
  swing2Motion=Free
  twistLimitDegrees=45
  twistMotion=Free
 }
 {
  bone1Name=spine_04
  bone2Name=spine_03
  swing1LimitDegrees=45
  swing1Motion=Free
  swing2LimitDegrees=45
  swing2Motion=Free
  twistLimitDegrees=45
  twistMotion=Free
 }
 {
  bone1Name=spine_05
  bone2Name=spine_04
  swing1LimitDegrees=45
  swing1Motion=Free
  swing2LimitDegrees=45
  swing2Motion=Free
  twistLimitDegrees=45
  twistMotion=Free
 }
 {
  bone1Name=clavicle_l
  bone2Name=spine_05
  swing1LimitDegrees=45
  swing1Motion=Free
  swing2LimitDegrees=45
  swing2Motion=Free
  twistLimitDegrees=45
  twistMotion=Free
 }
 {
  bone1Name=upperarm_l
  bone2Name=clavicle_l
  swing1LimitDegrees=45
  swing1Motion=Free
  swing2LimitDegrees=45
  swing2Motion=Free
  twistLimitDegrees=45
  twistMotion=Free
 }
 {
  bone1Name=lowerarm_l
  bone2Name=upperarm_l
  swing1LimitDegrees=45
  swing1Motion=Free
  swing2LimitDegrees=45
  swing2Motion=Free
  twistLimitDegrees=45
  twistMotion=Free
 }
 {
  bone1Name=neck_02
  bone2Name=spine_05
  swing1LimitDegrees=45
  swing1Motion=Free
  swing2LimitDegrees=45
  swing2Motion=Free
  twistLimitDegrees=45
  twistMotion=Free
 }
 {
  bone1Name=head
  bone2Name=neck_02
  swing1LimitDegrees=45
  swing1Motion=Free
  swing2LimitDegrees=45
  swing2Motion=Free
  twistLimitDegrees=45
  twistMotion=Free
 }
 {
  bone1Name=thigh_l
  bone2Name=pelvis
  swing1LimitDegrees=45
  swing1Motion=Free
  swing2LimitDegrees=45
  swing2Motion=Free
  twistLimitDegrees=45
  twistMotion=Free
 }
 {
  bone1Name=calf_l
  bone2Name=thigh_l
  swing1LimitDegrees=45
  swing1Motion=Free
  swing2LimitDegrees=45
  swing2Motion=Free
  twistLimitDegrees=45
  twistMotion=Free
 }
 {
  bone1Name=foot_l
  bone2Name=calf_l
  swing1LimitDegrees=45
  swing1Motion=Free
  swing2LimitDegrees=45
  swing2Motion=Free
  twistLimitDegrees=45
  twistMotion=Free
 }
 {
  bone1Name=clavicle_r
  bone2Name=spine_05
  swing1LimitDegrees=45
  swing1Motion=Free
  swing2LimitDegrees=45
  swing2Motion=Free
  twistLimitDegrees=45
  twistMotion=Free
 }
 {
  bone1Name=upperarm_r
  bone2Name=clavicle_r
  swing1LimitDegrees=45
  swing1Motion=Free
  swing2LimitDegrees=45
  swing2Motion=Free
  twistLimitDegrees=45
  twistMotion=Free
 }
 {
  bone1Name=lowerarm_r
  bone2Name=upperarm_r
  swing1LimitDegrees=45
  swing1Motion=Free
  swing2LimitDegrees=45
  swing2Motion=Free
  twistLimitDegrees=45
  twistMotion=Free
 }
 {
  bone1Name=thigh_r
  bone2Name=pelvis
  swing1LimitDegrees=45
  swing1Motion=Free
  swing2LimitDegrees=45
  swing2Motion=Free
  twistLimitDegrees=45
  twistMotion=Free
 }
 {
  bone1Name=calf_r
  bone2Name=thigh_r
  swing1LimitDegrees=45
  swing1Motion=Free
  swing2LimitDegrees=45
  swing2Motion=Free
  twistLimitDegrees=45
  twistMotion=Free
 }
 {
  bone1Name=foot_r
  bone2Name=calf_r
  swing1LimitDegrees=45
  swing1Motion=Free
  swing2LimitDegrees=45
  swing2Motion=Free
  twistLimitDegrees=45
  twistMotion=Free
 }
 {
  bone1Name=dyn_C_drape_l_back_1
  bone2Name=pelvis
  swing1LimitDegrees=45
  swing1Motion=Free
  swing2LimitDegrees=45
  swing2Motion=Free
  twistLimitDegrees=45
  twistMotion=Free
 }
 {
  bone1Name=dyn_C_drape_l_back_2
  bone2Name=dyn_C_drape_l_back_1
  swing1LimitDegrees=45
  swing1Motion=Free
  swing2LimitDegrees=45
  swing2Motion=Free
  twistLimitDegrees=45
  twistMotion=Free
 }
 {
  bone1Name=dyn_C_drape_l_back_3
  bone2Name=dyn_C_drape_l_back_2
  swing1LimitDegrees=45
  swing1Motion=Free
  swing2LimitDegrees=45
  swing2Motion=Free
  twistLimitDegrees=45
  twistMotion=Free
 }
 {
  bone1Name=dyn_C_drape_l_back_4
  bone2Name=dyn_C_drape_l_back_3
  swing1LimitDegrees=45
  swing1Motion=Free
  swing2LimitDegrees=45
  swing2Motion=Free
  twistLimitDegrees=45
  twistMotion=Free
 }
 {
  bone1Name=dyn_C_drape_r_back_1
  bone2Name=pelvis
  swing1LimitDegrees=45
  swing1Motion=Free
  swing2LimitDegrees=45
  swing2Motion=Free
  twistLimitDegrees=45
  twistMotion=Free
 }
 {
  bone1Name=dyn_C_drape_r_back_2
  bone2Name=dyn_C_drape_r_back_1
  swing1LimitDegrees=45
  swing1Motion=Free
  swing2LimitDegrees=45
  swing2Motion=Free
  twistLimitDegrees=45
  twistMotion=Free
 }
 {
  bone1Name=dyn_C_drape_r_back_3
  bone2Name=dyn_C_drape_r_back_2
  swing1LimitDegrees=45
  swing1Motion=Free
  swing2LimitDegrees=45
  swing2Motion=Free
  twistLimitDegrees=45
  twistMotion=Free
 }
 {
  bone1Name=dyn_C_drape_r_back_4
  bone2Name=dyn_C_drape_r_back_3
  swing1LimitDegrees=45
  swing1Motion=Free
  swing2LimitDegrees=45
  swing2Motion=Free
  twistLimitDegrees=45
  twistMotion=Free
 }
 {
  bone1Name=dyn_L_drape_side_1
  bone2Name=pelvis
  swing1LimitDegrees=45
  swing1Motion=Free
  swing2LimitDegrees=45
  swing2Motion=Free
  twistLimitDegrees=45
  twistMotion=Free
 }
 {
  bone1Name=dyn_L_drape_side_2
  bone2Name=dyn_L_drape_side_1
  swing1LimitDegrees=45
  swing1Motion=Free
  swing2LimitDegrees=45
  swing2Motion=Free
  twistLimitDegrees=45
  twistMotion=Free
 }
 {
  bone1Name=dyn_L_drape_side_3
  bone2Name=dyn_L_drape_side_2
  swing1LimitDegrees=45
  swing1Motion=Free
  swing2LimitDegrees=45
  swing2Motion=Free
  twistLimitDegrees=45
  twistMotion=Free
 }
 {
  bone1Name=dyn_L_drape_side_4
  bone2Name=dyn_L_drape_side_3
  swing1LimitDegrees=45
  swing1Motion=Free
  swing2LimitDegrees=45
  swing2Motion=Free
  twistLimitDegrees=45
  twistMotion=Free
 }
 {
  bone1Name=dyn_R_drape_side_1
  bone2Name=pelvis
  swing1LimitDegrees=45
  swing1Motion=Free
  swing2LimitDegrees=45
  swing2Motion=Free
  twistLimitDegrees=45
  twistMotion=Free
 }
 {
  bone1Name=dyn_R_drape_side_2
  bone2Name=dyn_R_drape_side_1
  swing1LimitDegrees=45
  swing1Motion=Free
  swing2LimitDegrees=45
  swing2Motion=Free
  twistLimitDegrees=45
  twistMotion=Free
 }
 {
  bone1Name=dyn_R_drape_side_3
  bone2Name=dyn_R_drape_side_2
  swing1LimitDegrees=45
  swing1Motion=Free
  swing2LimitDegrees=45
  swing2Motion=Free
  twistLimitDegrees=45
  twistMotion=Free
 }
 {
  bone1Name=dyn_R_drape_side_4
  bone2Name=dyn_R_drape_side_3
  swing1LimitDegrees=45
  swing1Motion=Free
  swing2LimitDegrees=45
  swing2Motion=Free
  twistLimitDegrees=45
  twistMotion=Free
 }
]
