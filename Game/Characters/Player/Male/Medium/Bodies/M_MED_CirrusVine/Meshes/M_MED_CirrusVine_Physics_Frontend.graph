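# Physics asset for the frontend version of the CirrusVine medium male player body, setting up its hanging drapes around the pelvis and legs.
# Six four-box drape chains have per-joint swing and twist limits: front and side chains hang from small master capsules angularly locked to the pelvis, back chains directly from the pelvis.
# Free-rotation constraints tie the two back chains together row by row.
# Pelvis, thigh and calf capsules form an angularly locked lower body.
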PhysicsAsset /Game/Characters/Player/Male/Medium/Bodies/M_MED_CirrusVine/Meshes/M_MED_CirrusVine_Physics_Frontend
bodies {
 calf_l [{
   shapeType=Capsule
   length=14.1259
   radius=9.0743
   shapeName=calf_l_capsule
   center=<-11.8948 -0.9372 1.7918>
   rotation=<rot 84.9779 -178.7634 178.1691>
  }]
 calf_r [{
   shapeType=Capsule
   length=14.1259
   radius=9.0743
   shapeName=calf_l_capsule
   center=<11.8948 0.9372 -1.7918>
   rotation=<rot 84.9779 -178.7634 178.1691>
  }]
 dyn_C_drape_l_back_1 [{shapeType=Box, extentX=17, extentY=4.7441, extentZ=15, shapeName=dyn_C_drape_l_back_1_box, center=<-5 0 0>, rotation=<rot 90 0 0>}]
 dyn_C_drape_l_back_2 [{shapeType=Box, extentX=18, extentY=3.8984, extentZ=15, shapeName=dyn_C_drape_l_back_2_box, center=<-3 0 0>, rotation=<rot 90 0 0>}]
 dyn_C_drape_l_back_3 [{shapeType=Box, extentX=17.8344, extentY=4.4782, extentZ=15, shapeName=dyn_C_drape_l_back_3_box, center=<-3 0 0>, rotation=<rot 90 0 0>}]
 dyn_C_drape_l_back_4 [{shapeType=Box, extentX=17.2673, extentY=2.4233, extentZ=7, shapeName=dyn_C_drape_l_back_4_box, center=<-3 0 0>, rotation=<rot 90 0 0>}]
 dyn_C_drape_r_back_1 [{shapeType=Box, extentX=17, extentY=4.7441, extentZ=15, shapeName=dyn_C_drape_r_back_1_box, center=<5 0 0>, rotation=<rot 90 0 0>}]
 dyn_C_drape_r_back_2 [{shapeType=Box, extentX=18, extentY=3.8984, extentZ=15, shapeName=dyn_C_drape_r_back_2_box, center=<3 0 0>, rotation=<rot 90 0 0>}]
 dyn_C_drape_r_back_3 [{shapeType=Box, extentX=17.8344, extentY=4.4782, extentZ=15, shapeName=dyn_C_drape_r_back_3_box, center=<3 0 0>, rotation=<rot 90 0 0>}]
 dyn_C_drape_r_back_4 [{shapeType=Box, extentX=17.2673, extentY=2.4233, extentZ=7, shapeName=dyn_C_drape_r_back_4_box, center=<3 0 0>, rotation=<rot 90 0 0>}]
 dyn_L_drape_front_1 [{shapeType=Box, extentX=15, extentY=2.0677, extentZ=11.7427, shapeName=dyn_L_drape_front_1_box, center=<-6 0 0>, rotation=<rot 90 0 0>}]
 dyn_L_drape_front_2 [{shapeType=Box, extentX=15, extentY=2.9256, extentZ=11.3806, shapeName=dyn_L_drape_front_2_box, center=<-3 0 0>, rotation=<rot 90 0 0>}]
 dyn_L_drape_front_3 [{shapeType=Box, extentX=15, extentY=4.7455, extentZ=13, shapeName=dyn_L_drape_front_3_box, center=<-3 0 0>, rotation=<rot 90 0 0>}]
 dyn_L_drape_front_4 [{shapeType=Box, extentX=15, extentY=6.0416, extentZ=12.6096, shapeName=dyn_L_drape_front_4_box, center=<-3 0 0>, rotation=<rot 90 0 0>}]
 dyn_L_drape_front_master [{shapeType=Capsule, length=0.505, radius=0.505, shapeName=dyn_L_drape_front_master_capsule}]
 dyn_L_drape_side_1 [{shapeType=Box, extentX=15, extentY=3.4648, extentZ=15, shapeName=dyn_L_drape_side_1_box, center=<-6 0 0>, rotation=<rot 90 0 0>}]
 dyn_L_drape_side_2 [{shapeType=Box, extentX=15, extentY=3.946, extentZ=15, shapeName=dyn_L_drape_side_2_box, center=<-3 0 0>, rotation=<rot 90 0 0>}]
 dyn_L_drape_side_3 [{
   shapeType=Box
   extentX=15
   extentY=4.0387
   extentZ=15
   shapeName=dyn_L_drape_side_3_box
   center=<-3 0 0>
   rotation=<rot 90 169.1228 174.1228>
  }]
 dyn_L_drape_side_4 [{shapeType=Box, extentX=15, extentY=5.9363, extentZ=15, shapeName=dyn_L_drape_side_4_box, center=<-3 0 0>, rotation=<rot 90 0 0>}]
 dyn_L_drape_side_master [{shapeType=Capsule, length=0.505, radius=0.505, shapeName=dyn_L_drape_side_master_capsule}]
 dyn_R_drape_front_1 [{shapeType=Box, extentX=15, extentY=2.0677, extentZ=11.7427, shapeName=dyn_R_drape_front_1_box, center=<6 0 0>, rotation=<rot 90 0 0>}]
 dyn_R_drape_front_2 [{shapeType=Box, extentX=15, extentY=2.9256, extentZ=11.3806, shapeName=dyn_R_drape_front_2_box, center=<3 0 0>, rotation=<rot 90 0 0>}]
 dyn_R_drape_front_3 [{shapeType=Box, extentX=15, extentY=4.7455, extentZ=13, shapeName=dyn_R_drape_front_3_box, center=<3 0 0>, rotation=<rot 90 0 0>}]
 dyn_R_drape_front_4 [{shapeType=Box, extentX=15, extentY=6.0416, extentZ=12.6096, shapeName=dyn_R_drape_front_4_box, center=<3 0 0>, rotation=<rot 90 0 0>}]
 dyn_R_drape_front_master [{shapeType=Capsule, length=0.505, radius=0.505, shapeName=dyn_R_drape_front_master_capsule}]
 dyn_R_drape_side_1 [{shapeType=Box, extentX=15, extentY=3.4648, extentZ=15, shapeName=dyn_R_drape_side_1_box, center=<6 0.7925 0>, rotation=<rot 90 0 0>}]
 dyn_R_drape_side_2 [{shapeType=Box, extentX=15, extentY=3.946, extentZ=15, shapeName=dyn_R_drape_side_2_box, center=<3 0 0>, rotation=<rot 90 0 0>}]
 dyn_R_drape_side_3 [{shapeType=Box, extentX=15, extentY=4.0387, extentZ=15, shapeName=dyn_R_drape_side_3_box, center=<3 0 0>, rotation=<rot 90 169.1228 174.1228>}]
 dyn_R_drape_side_4 [{shapeType=Box, extentX=15, extentY=5.9363, extentZ=15, shapeName=dyn_R_drape_side_4_box, center=<3 0 0>, rotation=<rot 90 0 0>}]
 dyn_R_drape_side_master [{shapeType=Capsule, length=0.505, radius=0.505, shapeName=dyn_R_drape_side_master_capsule}]
 pelvis [{shapeType=Capsule, length=5, radius=9.5, shapeName=pelvis_capsule, center=<-3 0 0>}]
 thigh_l [{shapeType=Capsule, length=31.448, radius=9, shapeName=thigh_l_capsule, center=<-22.0023 -0.469 2.4999>, rotation=<rot 90 0 0>}]
 thigh_r [{shapeType=Capsule, length=28.1041, radius=9, shapeName=thigh_l_capsule, center=<22.0037 0.7578 -2.4999>, rotation=<rot 90 0 0>}]
}
constraints [
 {bone1Name=dyn_C_drape_l_back_1, bone2Name=pelvis, swing1LimitDegrees=40, swing2LimitDegrees=10, twistLimitDegrees=10}
 {bone1Name=dyn_C_drape_l_back_2, bone2Name=dyn_C_drape_l_back_1, swing1LimitDegrees=40, swing2LimitDegrees=10, twistLimitDegrees=10}
 {bone1Name=dyn_C_drape_l_back_3, bone2Name=dyn_C_drape_l_back_2, swing1LimitDegrees=40, swing2LimitDegrees=10, twistLimitDegrees=10}
 {bone1Name=dyn_C_drape_l_back_4, bone2Name=dyn_C_drape_l_back_3, swing1LimitDegrees=40, swing2LimitDegrees=10, twistLimitDegrees=10}
 {bone1Name=dyn_C_drape_r_back_1, bone2Name=pelvis, swing1LimitDegrees=40, swing2LimitDegrees=10, twistLimitDegrees=10}
 {bone1Name=dyn_C_drape_r_back_2, bone2Name=dyn_C_drape_r_back_1, swing1LimitDegrees=40, swing2LimitDegrees=10, twistLimitDegrees=10}
 {bone1Name=dyn_C_drape_r_back_3, bone2Name=dyn_C_drape_r_back_2, swing1LimitDegrees=40, swing2LimitDegrees=10, twistLimitDegrees=10}
 {bone1Name=dyn_C_drape_r_back_4, bone2Name=dyn_C_drape_r_back_3, swing1LimitDegrees=40, swing2LimitDegrees=10, twistLimitDegrees=10}
 {bone1Name=dyn_L_drape_front_2, bone2Name=dyn_L_drape_front_1, swing1LimitDegrees=45, swing2LimitDegrees=10, twistLimitDegrees=10}
 {bone1Name=dyn_L_drape_front_3, bone2Name=dyn_L_drape_front_2, swing1LimitDegrees=50, swing2LimitDegrees=10, twistLimitDegrees=10}
 {bone1Name=dyn_L_drape_front_4, bone2Name=dyn_L_drape_front_3, swing1LimitDegrees=50, swing2LimitDegrees=10, twistLimitDegrees=10}
 {bone1Name=dyn_L_drape_side_2, bone2Name=dyn_L_drape_side_1, swing1LimitDegrees=30, swing2LimitDegrees=10, twistLimitDegrees=10}
 {bone1Name=dyn_L_drape_side_3, bone2Name=dyn_L_drape_side_2, swing1LimitDegrees=30, swing2LimitDegrees=10, twistLimitDegrees=10}
 {bone1Name=dyn_L_drape_side_4, bone2Name=dyn_L_drape_side_3, swing1LimitDegrees=30, swing2LimitDegrees=10, twistLimitDegrees=10}
 {bone1Name=dyn_R_drape_front_2, bone2Name=dyn_R_drape_front_1, swing1LimitDegrees=45, swing2LimitDegrees=10, twistLimitDegrees=10}
 {bone1Name=dyn_R_drape_front_3, bone2Name=dyn_R_drape_front_2, swing1LimitDegrees=50, swing2LimitDegrees=10, twistLimitDegrees=10}
 {bone1Name=dyn_R_drape_front_4, bone2Name=dyn_R_drape_front_3, swing1LimitDegrees=50, swing2LimitDegrees=10, twistLimitDegrees=10}
 {bone1Name=dyn_R_drape_side_2, bone2Name=dyn_R_drape_side_1, swing1LimitDegrees=30, swing2LimitDegrees=20, twistLimitDegrees=10}
 {bone1Name=dyn_R_drape_side_3, bone2Name=dyn_R_drape_side_2, swing1LimitDegrees=30, swing2LimitDegrees=20, twistLimitDegrees=10}
 {bone1Name=dyn_R_drape_side_4, bone2Name=dyn_R_drape_side_3, swing1LimitDegrees=30, swing2LimitDegrees=20, twistLimitDegrees=10}
 {
  bone1Name=thigh_l
  bone2Name=pelvis
  swing1LimitDegrees=45
  swing1Motion=Locked
  swing2LimitDegrees=45
  swing2Motion=Locked
  twistLimitDegrees=45
  twistMotion=Locked
 }
 {
  bone1Name=calf_l
  bone2Name=thigh_l
  swing1LimitDegrees=45
  swing1Motion=Locked
  swing2LimitDegrees=45
  swing2Motion=Locked
  twistLimitDegrees=45
  twistMotion=Locked
 }
 {
  bone1Name=thigh_r
  bone2Name=pelvis
  swing1LimitDegrees=45
  swing1Motion=Locked
  swing2LimitDegrees=45
  swing2Motion=Locked
  twistLimitDegrees=45
  twistMotion=Locked
 }
 {
  bone1Name=calf_r
  bone2Name=thigh_r
  swing1LimitDegrees=45
  swing1Motion=Locked
  swing2LimitDegrees=45
  swing2Motion=Locked
  twistLimitDegrees=45
  twistMotion=Locked
 }
 {
  bone1Name=dyn_L_drape_front_master
  bone2Name=pelvis
  swing1LimitDegrees=45
  swing1Motion=Locked
  swing2LimitDegrees=45
  swing2Motion=Locked
  twistLimitDegrees=45
  twistMotion=Locked
 }
 {
  bone1Name=dyn_L_drape_side_master
  bone2Name=pelvis
  swing1LimitDegrees=45
  swing1Motion=Locked
  swing2LimitDegrees=45
  swing2Motion=Locked
  twistLimitDegrees=45
  twistMotion=Locked
 }
 {
  bone1Name=dyn_R_drape_front_master
  bone2Name=pelvis
  swing1LimitDegrees=45
  swing1Motion=Locked
  swing2LimitDegrees=45
  swing2Motion=Locked
  twistLimitDegrees=45
  twistMotion=Locked
 }
 {
  bone1Name=dyn_R_drape_side_master
  bone2Name=pelvis
  swing1LimitDegrees=45
  swing1Motion=Locked
  swing2LimitDegrees=45
  swing2Motion=Locked
  twistLimitDegrees=45
  twistMotion=Locked
 }
 {bone1Name=dyn_L_drape_front_1, bone2Name=dyn_L_drape_front_master, swing1LimitDegrees=30, swing2LimitDegrees=10, twistLimitDegrees=10}
 {bone1Name=dyn_L_drape_side_1, bone2Name=dyn_L_drape_side_master, swing1LimitDegrees=30, swing2LimitDegrees=10, twistLimitDegrees=10}
 {bone1Name=dyn_R_drape_front_1, bone2Name=dyn_R_drape_front_master, swing1LimitDegrees=30, swing2LimitDegrees=10, twistLimitDegrees=10}
 {bone1Name=dyn_R_drape_side_1, bone2Name=dyn_R_drape_side_master, swing1LimitDegrees=30, swing2LimitDegrees=20, twistLimitDegrees=10}
 {
  bone1Name=dyn_C_drape_r_back_4
  bone2Name=dyn_C_drape_l_back_4
  swing1LimitDegrees=45
  swing1Motion=Free
  swing2LimitDegrees=45
  swing2Motion=Free
  twistLimitDegrees=45
  twistMotion=Free
 }
 {
  bone1Name=dyn_C_drape_r_back_3
  bone2Name=dyn_C_drape_l_back_3
  swing1LimitDegrees=45
  swing1Motion=Free
  swing2LimitDegrees=45
  swing2Motion=Free
  twistLimitDegrees=45
  twistMotion=Free
 }
 {
  bone1Name=dyn_C_drape_r_back_2
  bone2Name=dyn_C_drape_l_back_2
  swing1LimitDegrees=45
  swing1Motion=Free
  swing2LimitDegrees=45
  swing2Motion=Free
  twistLimitDegrees=45
  twistMotion=Free
 }
 {
  bone1Name=dyn_C_drape_r_back_1
  bone2Name=dyn_C_drape_l_back_1
  swing1LimitDegrees=45
  swing1Motion=Free
  swing2LimitDegrees=45
  swing2Motion=Free
  twistLimitDegrees=45
  twistMotion=Free
 }
]
